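# Niagara idle effect for the LiftingAura pickaxe: a reddish side glow made by one burst-spawn emitter with two sprite renderers.
# Sprites have fixed facing and alignment, are offset from the origin, and store a mirrored BackPosition.
# Color intensity follows time of day through alpha, which a curve also modulates.
# Particles are not killed when their lifetime ends, and sprite size scales with the owner.
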NiagaraSystem /BRCosmetics/Weapons/FORT_Melee/Pickaxe_LiftingAura/FX/NS_Pickaxe_LiftingAura_Idle
dependencies {usedDataInterfaces [&NiagaraDataInterfaceCurve], usedDynamicInputs [], usedModules [&SystemState]}
emitters {Glow_Side001 {
  emitterUpdateScript [
   &EmitterState(
    Inactive Response=::NewEnumerator1
    Life Cycle Mode=::NewEnumerator1
    Loop Behavior=::NewEnumerator1
    Loop Duration Mode=::NewEnumerator1
    UseLoopDelay=false
   )
   &SpawnBurst_Instantaneous(Use Spawn Probability=false, UseLoopCountLimit=false)
  ]
  particleSpawnScript [
   &InitializeParticle(
    Color=&FNTimeOfDayColor(
     Night Color=<rgba 2 0 0.0326 0.25>
     Morning Color=<rgba 2 0 0.0326 0.75>
     Day Color=<rgba 2 0 0.0326 1>
     Evening Color=<rgba 2 0 0.0326 0.75>
    )
    Lifetime=2
    Position Offset=<0 -2 12>
    Sprite Rotation Angle=1
    Sprite Rotation Mode=::NewEnumerator0
    Sprite Size Mode=::NewEnumerator3
    Uniform Sprite Size=14
    UsePositionOffset=true
    Write Material Random=true
   )
   &SpriteFacingAndAlignment(Sprite Alignment=<0 0 1>, Sprite Facing=<0 1 0>, WriteAlignment=true, WriteFacing=true)
   &CameraOffset(Camera Offset Amount=0.1)
   /SetVariables(Particles.BackPosition:NiagaraPosition=&Multiply_Vector(A=<0 2 12>, B=<1 1 1>)) {isSetParameters=true}
  ]
  particleUpdateScript [
   &ParticleState(Kill Particles When Lifetime Has Elapsed=false)
   &ScaleSpriteSize(
    Scale Sprite Size Mode=::NewEnumerator0
    Uniform Scale Factor=&MakeFloatFromVector(Vector=$Engine.Owner.Scale, Channel=::NewEnumerator0)
   )
   &ScaleColor(
    Scale Alpha=&FloatFromCurve(
     FloatCurve=DataInterface() {propertyValues {Curve {keys [<at 0 0.6 RCIM_Cubic>,<at 0.3 1 RCIM_Cubic>,<at 1 0.6 RCIM_Cubic>]}}}
     Scale Curve=0.2
    )
    ScaleA=true
    ScaleRGB=true
   )
  ]
  rendererClasses [&NiagaraSpriteRendererProperties]
  renderers [
   {rendererClass=&NiagaraSpriteRendererProperties}
   {rendererClass=&NiagaraSpriteRendererProperties, rendererIndex=1}
  ]
 }}
summary {
 systemName=NS_Pickaxe_LiftingAura_Idle
 emitters [{emitterName=Glow_Side001, rendererClasses [&NiagaraSpriteRendererProperties]}]
 userVariables [
  User.bIsFrontEnd {type=NiagaraBool}
  User.bIsFrontEndPreview {type=NiagaraBool}
 ]
}
