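# Niagara trail effect for a variant of the LiftingAura melee pickaxe, layering seven short timed emitters on a skeletal mesh.
# Five ribbon emitters use AnimTrail modules between mesh sockets for dark, colored, fast and three-strand screen trails, faded and tinted by curves.
# Gold_Dust and Snowflakes002 spawn sprites at a socket, stirred by curl noise and gravity; the snowflakes also spawn by distance moved.
# Time-of-day inputs adjust several colors and opacities.
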NiagaraSystem /BRCosmetics/Weapons/FORT_Melee/Pickaxe_LiftingAura/FX/NS_Pickaxe_LiftingAura_Trail_Variant
dependencies {
 usedDataInterfaces [
  &NiagaraDataInterfaceCurve
  &NiagaraDataInterfaceColorCurve
  &NiagaraDataInterfaceMeshRendererInfo
  &NiagaraDataInterfaceSkeletalMesh
 ]
 usedDynamicInputs []
 usedModules [
  &self:SystemSpawnScript.SystemScriptSource.SystemScriptGraph.NiagaraNodeAssignment_0.SetVariables_#1
  &SystemState
 ]
}
emitters {
 Dark001 {
  emitterUpdateScript [
   &EmitterState(Life Cycle Mode=::NewEnumerator1, Loop Behavior=::NewEnumerator1, Loop Delay=0.1, Loop Duration=$System.FXLifeLength, UseLoopDelay=true)
   &FN_SpawnAnimTrail_v2(Number of Anim Trails:NiagaraInt32=&Multiply_Int(A=1, B=1))
  ]
  particleSpawnScript [
   /Game/Effects/Niagara/Modules/AnimTrail/FN_InitializeAnimTrail/FN_InitializeAnimTrailBoneIndecies(Second Bone or Socket Index=1)
   &FN_InitializeAnimTrail(Lifetime=0.15, SkeletalMesh=$System.Skeletal Mesh) {module=FN_InitializeAnimTrail001}
  ]
  particleUpdateScript [
   &FN_UpdateAnimTrail(SkeletalMesh=$System.Skeletal Mesh)
   &ScaleRibbonWidth(Ribbon Width Scale=&FloatFromCurve(FloatCurve=DataInterface() {propertyValues {Curve {keys [{arriveTangent=-0.8703, time=0.2, value=1}, <at 1 0.8>]}}}))
   &ScaleColor(
    Scale Alpha=&FloatFromCurve(
     FloatCurve=DataInterface() {propertyValues {Curve {keys [
         {arriveTangent=-0.2226, interpMode=RCIM_Cubic, leaveTangent=-0.2226, tangentMode=RCTM_User, time=0.1, value=1}
         {arriveTangent=-0.111, interpMode=RCIM_Cubic, leaveTangent=-0.111, tangentMode=RCTM_User, time=0.3, value=0.25}
         {arriveTangent=-0.0236, tangentMode=RCTM_User, time=1, value=0}
        ]}}}
     Scale Curve=5
    )
    Scale RGB=&MakeVectorFromLinearColorRGB(Color=&ColorFromCurve(ColorCurve=DataInterface() {propertyValues {
        AlphaCurve {keys [<at 0 1 RCIM_Cubic>]}
        BlueCurve {keys [
          <at 0.0015 0.106>
          {arriveTangent=-0.124, interpMode=RCIM_Cubic, leaveTangent=-0.124, time=0.1681, value=0.0845}
          {arriveTangent=-0.0418, interpMode=RCIM_Cubic, leaveTangent=-0.0418, time=0.4528, value=0.05}
          <at 0.9946 0.05>
         ]}
        GreenCurve {keys [
          <at 0.0015 0.011>
          {arriveTangent=-0.0155, interpMode=RCIM_Cubic, leaveTangent=-0.0155, time=0.1681, value=0.0073}
          {arriveTangent=-0.0019, interpMode=RCIM_Cubic, leaveTangent=-0.0019, time=0.4528, value=0.0039}
          <at 0.9946 0.0058>
         ]}
        RedCurve {keys [
          <at 0.0015 0.15>
          {arriveTangent=-0.2403, interpMode=RCIM_Cubic, leaveTangent=-0.2403, time=0.1681, value=0.1}
          {arriveTangent=-0.0882, interpMode=RCIM_Cubic, leaveTangent=-0.0882, time=0.4528, value=0.0416}
          <at 0.9946 0.0271>
         ]}
       }}))
    ScaleA=true
    ScaleRGB=true
   )
   &SpriteFacingAndAlignment(Sprite Facing=$Particles.RibbonFacing, WriteAlignment=false, WriteFacing=true)
   &ScaleColor(
    Scale Alpha=&FloatFromCurve(
     FloatCurve=DataInterface() {propertyValues {Curve {keys [
         {tangentMode=RCTM_User, time=0, value=1}
         {arriveTangent=-0.0405, interpMode=RCIM_Cubic, leaveTangent=-0.0405, tangentMode=RCTM_User, time=0.3, value=1}
         {arriveTangent=-3.8778, interpMode=RCIM_Cubic, leaveTangent=-3.8778, tangentMode=RCTM_User, time=0.7, value=0.075}
        ]}}}
     CurveIndex=$Emitter.NormalizedLoopAge
    )
    ScaleA=true
    ScaleRGB=true
   ) {module=ScaleColor001}
  ]
  rendererClasses [&NiagaraRibbonRendererProperties]
  renderers [{rendererClass=&NiagaraRibbonRendererProperties}]
 }
 Gold_Dust {
  emitterUpdateScript [
   &EmitterState(Life Cycle Mode=::NewEnumerator1, Loop Behavior=::NewEnumerator1, Loop Delay=0.04, Loop Duration=0.15, UseLoopDelay=true)
   &SpawnRate(
    Spawn Probability=&RandomRangeFloat(Minimum=0.5, Maximum=1.25, Evaluation Type=::NewEnumerator1)
    SpawnRate=&FloatFromCurve(
     FloatCurve=DataInterface() {propertyValues {Curve {keys [
         {arriveTangent=1.7804, interpMode=RCIM_Cubic, leaveTangent=1.7804, tangentMode=RCTM_User, time=0, value=0.2}
         {interpMode=RCIM_Cubic, tangentMode=RCTM_User, time=0.8, value=1}
        ]}}}
     CurveIndex=$Emitter.NormalizedLoopAge
     Scale Curve=150
    )
    Use Spawn Probability=true
   )
  ]
  particleSpawnScript [
   &InitializeParticle(
    Color=&FNTimeOfDayColor(
     Night Color=<rgba 0.4 0.2 0.08 1>
     Morning Color=<rgba 0.9 0.495 0.09 1>
     Day Color=<rgba 1 0.55 0.1 1>
     Evening Color=<rgba 0.9 0.495 0.09 1>
    )
    Lifetime Max=0.2
    Lifetime Min=0.1
    Lifetime Mode=::NewEnumerator1
    Sprite Rotation Angle=45
    Sprite Rotation Mode=::NewEnumerator0
    Sprite Size Mode=::NewEnumerator1
    Sprite UV Mode=::NewEnumerator3
    Uniform Sprite Size Max=8
    UsePositionOffset=false
    Write Material Random=true
   )
   &SkeletalMeshLocation(
    Bone / Socket Index=1
    MaskSpawnGroup=false
    Mesh Sampling Type=::NewEnumerator5
    Skeletal Mesh=$System.Skeletal Mesh
    Socket Sampling Mode=::NewEnumerator5
   ) {module=SkeletalMeshLocation001}
   &SphereLocation(
    MaskSpawnGroup=false
    Offset=&RandomRangeVector(Minimum=<-10 0 0>, Maximum=<10 0 0>)
    Override Seed=false
    Sphere Radius=10
    Surface Expansion Mode=::NewEnumerator1
    Surface Only=true
    UseNonUniformScale=false
    UseOffset=true
    UseSphereOrientation=false
   )
  ]
  particleUpdateScript [
   &ParticleState()
   &CurlNoiseForce(
    Noise Frequency=30
    Noise Strength=&FloatFromCurve(
     FloatCurve=DataInterface() {propertyValues {Curve {keys [
         {arriveTangent=-2.3694, interpMode=RCIM_Cubic, leaveTangent=-2.3694, tangentMode=RCTM_User, time=0, value=1}
         {interpMode=RCIM_Cubic, tangentMode=RCTM_User, time=1, value=0}
        ]}}}
     Scale Curve=700
    )
    Pan Noise=true
    Pan Noise Field=<0 0 1>
   )
   &GravityForce(Gravity=<0 0 -128>)
   &SolveForcesAndVelocity(Clamp Velocity=false, Limit Acceleration=false)
   &ScaleSpriteSize(
    Uniform Curve Scale=&Waveform(
     Global Phase=$Particles.Age
     "Waveform [1]"=::NewEnumerator0
     "[1] Phase Scale"=&FloatFromCurve(FloatCurve=DataInterface() {propertyValues {Curve {keys [<at 0 1>,<at 1 0>]}}})
     "[1] Phase"=$Particles.Age
     "[1] Frequency"=30
     "[1] Amplitude Min/Max"=<0.25 1>
    )
    Uniform Curve Sprite Scale=DataInterface() {propertyValues {Curve {keys [<at 0 1 RCIM_Cubic>, {arriveTangent=-2.1124, tangentMode=RCTM_User, time=1, value=0}]}}}
   )
  ]
  rendererClasses [&NiagaraSpriteRendererProperties]
  renderers [{rendererClass=&NiagaraSpriteRendererProperties}]
 }
 Snowflakes002 {
  emitterUpdateScript [
   &EmitterState(Life Cycle Mode=::NewEnumerator1, Loop Behavior=::NewEnumerator1, Loop Duration=0.5, UseLoopDelay=false)
   &SpawnRate(SpawnRate=&RandomRangeFloat(Minimum=10, Maximum=15), Use Spawn Probability=false)
   &SpawnPerUnit(Module.Use Max Movement Threshold=true, Module.Use Movement Tolerance=false, Spawn Spacing=80, Use Spawn Probability=false)
  ]
  particleSpawnScript [
   &InitializeParticle(
    Color=&MultiplyLinearColor(
     Linear Color=&RandomRangeLinearColor(Minimum=<rgba 0.4375 0 0.1385 1>, Maximum=<rgba 0.2917 0 0.0729 1>)
     Scale Factor=&MakeVector4_FromFloat(Float=&FNTimeOfDayScalar(Night Scale=1, Morning Scale=1, Evening Scale=1))
    )
    Lifetime Max=0.45
    Lifetime Min=0.35
    Lifetime Mode=::NewEnumerator1
    Mass Mode=::NewEnumerator1
    Sprite Size Mode=::NewEnumerator1
    Sprite UV Mode=::NewEnumerator1
    Uniform Sprite Size Max=12
    Uniform Sprite Size Min=8
    UsePositionOffset=false
    Write Material Random=true
   ) {module=InitializeParticle001}
   &SkeletalMeshLocation(
    Bone / Socket Index=1
    "Inherit Velocity (Scale)"=0.1
    MaskSpawnGroup=false
    Mesh Sampling Type=::NewEnumerator5
    Orientation Sampling=::NewEnumerator1
    Skeletal Mesh=$System.Skeletal Mesh
    Socket Sampling Mode=::NewEnumerator5
    Velocity Sampling=::NewEnumerator0
    VelocitySpeedLimit=false
   )
   &InitialMeshOrientation(
    Mesh Orientation Mode=::NewEnumerator6
    Rotation=&RandomRangeVector(Minimum=<-0.05 0 0>, Maximum=<0.05 0 0>)
    Use Rotation Vector=true
   ) {module=InitialMeshOrientation001}
   /SetVariables(Particles.UVRandom:NiagaraFloat=&RandomRangeFloat(Minimum=0, Maximum=3)) {isSetParameters=true}
   &ApplyInitialForces()
   /SetVariables(Particles.SpriteScale:Vector2f=&Multiply_Vector2D(A=$Particles.SpriteSize, B=<0.95 0.95>)) {isSetParameters=true}
  ]
  particleUpdateScript [
   &ParticleState()
   &ScaleSpriteSize(Uniform Curve Sprite Scale=DataInterface() {propertyValues {Curve {keys [
        {arriveTangent=3.3477, interpMode=RCIM_Cubic, leaveTangent=3.3477, tangentMode=RCTM_User, time=0, value=0}
        {arriveTangent=-0.0944, interpMode=RCIM_Cubic, leaveTangent=-0.0944, tangentMode=RCTM_User, time=0.5, value=0.9847}
        {arriveTangent=-0.0405, interpMode=RCIM_Cubic, leaveTangent=-0.0405, tangentMode=RCTM_User, time=1, value=0.25}
       ]}}}) {module=ScaleSpriteSize002}
   /Niagara/Modules/Update/Forces/CurlNoiseForce(Noise Strength=&RandomRangeFloat(Minimum=600, Maximum=800), Pan Noise=false)
   &GravityForce(Gravity=&RandomRangeVector(Minimum=<0 0 -25>, Maximum=<0 0 -50>))
   &Drag(Drag=&UniformRangedFloat(Minimum=6, Maximum=8), Use Linear Drag=true, Use Rotational Drag=false)
   &SolveForcesAndVelocity(Clamp Velocity=false, Limit Acceleration=false)
   /SetVariables(Particles.SpriteColor=&MultiplyLinearColor(
     Linear Color=&FNTimeOfDayColor(Night Color=<rgba 1 1 1 1>, Morning Color=<rgba 1 1 1 1>, Day Color=<rgba 1 1 1 1>, Evening Color=<rgba 1 1 1 1>)
     Scale Factor=&MakeVector4_FromLinearColor(LinearColor=&ColorFromCurve(ColorCurve=DataInterface() {propertyValues {
         AlphaCurve {keys [
           {arriveTangent=2.6759, interpMode=RCIM_Cubic, leaveTangent=2.6759, tangentMode=RCTM_User, time=0, value=0.6}
           {arriveTangent=-1.7996, interpMode=RCIM_Cubic, leaveTangent=-2.7184, tangentMode=RCTM_Break, time=0.5, value=0.6}
           <at 1 0 RCIM_Cubic>
          ]}
         BlueCurve {keys [<at 0 2 RCIM_Cubic>,<at 0.4702 1>,<at 1 0.2604 RCIM_Cubic>]}
         GreenCurve {keys [<at 0 0.391 RCIM_Cubic>,<at 0.4702 0.2266>,<at 1 0.0396 RCIM_Cubic>]}
         RedCurve {keys [<at 0 1.2223 RCIM_Cubic>,<at 0.4702 0.5875>,<at 1 0.1758 RCIM_Cubic>]}
        }}))
    )) {isSetParameters=true}
   &DynamicMaterialParameters(Module.Index0Param0Write=true, Module.Index0Param1Write=true, Module.Index0Param2Write=true, Module.Index0Param3Write=true)
  ]
  rendererClasses [&NiagaraSpriteRendererProperties]
  renderers [{rendererClass=&NiagaraSpriteRendererProperties}]
 }
 Trail {
  emitterUpdateScript [
   &EmitterState(Life Cycle Mode=::NewEnumerator1, Loop Behavior=::NewEnumerator1, Loop Delay=0.1, Loop Duration=0.1, UseLoopDelay=true)
   &FN_SpawnAnimTrail_v2()
  ]
  particleSpawnScript [
   &FN_InitializeAnimTrailBoneIndecies(Second Bone or Socket Index=1)
   &FN_InitializeAnimTrail(Lifetime=0.23, SkeletalMesh=$System.Skeletal Mesh) {module=FN_InitializeAnimTrail001}
  ]
  particleUpdateScript [
   &FN_UpdateAnimTrail(SkeletalMesh=$System.Skeletal Mesh)
   &ScaleColor(
    Scale Alpha=&Multiply_Float(
     A=&FloatFromCurve(
      FloatCurve=DataInterface() {propertyValues {Curve {keys [{arriveTangent=-2.2326, interpMode=RCIM_Cubic, leaveTangent=-2.2326, tangentMode=RCTM_User, time=0, value=1}, <at 1 0>]}}}
      Scale Curve=0.7
     )
     B=1
    )
    Scale RGB=&MakeVectorFromLinearColorRGB(Color=&ColorFromCurve(ColorCurve=DataInterface() {propertyValues {
        AlphaCurve {keys [<at 0 1 RCIM_Cubic>]}
        BlueCurve {keys [
          <at 0.0015 0.6572>
          {arriveTangent=-0.0338, interpMode=RCIM_Cubic, leaveTangent=-0.0338, time=0.2137, value=0.4589}
          {arriveTangent=0.1807, interpMode=RCIM_Cubic, leaveTangent=0.1807, time=0.5113, value=0.64}
          <at 0.9946 0.6>
         ]}
        GreenCurve {keys [
          <at 0.0015 0.3165>
          {arriveTangent=-0.3995, interpMode=RCIM_Cubic, leaveTangent=-0.3995, time=0.2137, value=0.0956}
          {arriveTangent=-0.0093, interpMode=RCIM_Cubic, leaveTangent=-0.0093, time=0.5113, value=0.1128}
          <at 0.9946 0.0884>
         ]}
        RedCurve {keys [
          <at 0.0015 2>
          {arriveTangent=-3.4114, interpMode=RCIM_Cubic, leaveTangent=-3.4114, time=0.2137, value=0.75}
          {arriveTangent=-0.7961, interpMode=RCIM_Cubic, leaveTangent=-0.7961, time=0.5113, value=0.2608}
          <at 0.9946 0.1283>
         ]}
       }}))
    ScaleA=true
    ScaleRGB=true
   )
   &ScaleColor(
    Scale Alpha=&Multiply_Float(
     A=&FloatFromCurve(
      FloatCurve=DataInterface() {propertyValues {Curve {keys [{arriveTangent=-2.2326, interpMode=RCIM_Cubic, leaveTangent=-2.2326, tangentMode=RCTM_User, time=0, value=1}, <at 1 0>]}}}
      Scale Curve=0.7
     )
     B=1
    )
    Scale RGB=&MakeVectorFromLinearColorRGB(Color=&ColorFromCurve(ColorCurve=DataInterface() {propertyValues {
        AlphaCurve {keys [<at 0 1 RCIM_Cubic>]}
        BlueCurve {keys [
          <at 0.0015 0.2387>
          {arriveTangent=0.7872, interpMode=RCIM_Cubic, leaveTangent=0.7872, time=0.2137, value=0.434}
          {arriveTangent=0.2126, interpMode=RCIM_Cubic, leaveTangent=0.2126, time=0.5113, value=0.64}
          <at 0.9946 0.6>
         ]}
        GreenCurve {keys [
          <at 0.0015 0.146>
          {arriveTangent=-0.1872, interpMode=RCIM_Cubic, leaveTangent=-0.1872, time=0.2137, value=0.1095}
          {arriveTangent=-0.0519, interpMode=RCIM_Cubic, leaveTangent=-0.0519, time=0.5113, value=0.0506}
          <at 0.9946 0.069>
         ]}
        RedCurve {keys [
          <at 0.0015 2>
          {arriveTangent=-2.6677, interpMode=RCIM_Cubic, leaveTangent=-2.6677, time=0.2137, value=1.5}
          {arriveTangent=-1.5038, interpMode=RCIM_Cubic, leaveTangent=-1.5038, time=0.5113, value=0.64}
          <at 0.9946 0.3257>
         ]}
       }}))
    ScaleA=true
    ScaleRGB=true
   ) {enabled=false, module=ScaleColor001}
  ]
  rendererClasses [&NiagaraRibbonRendererProperties]
  renderers [{rendererClass=&NiagaraRibbonRendererProperties}]
 }
 Trail002 {
  emitterUpdateScript [
   &EmitterState(Life Cycle Mode=::NewEnumerator1, Loop Behavior=::NewEnumerator1, Loop Duration=0.27, UseLoopDelay=true)
   &FN_SpawnAnimTrail_v2()
  ]
  particleSpawnScript [
   &FN_InitializeAnimTrailBoneIndecies()
   &FN_InitializeAnimTrail(AnimTrailWidth=20, Lifetime=0.25, SkeletalMesh=$System.Skeletal Mesh) {module=FN_InitializeAnimTrail001}
  ]
  particleUpdateScript [
   &FN_UpdateAnimTrail(SkeletalMesh=$System.Skeletal Mesh)
   &ScaleColor(
    Scale Alpha=&Multiply_Float(
     A=&FloatFromCurve(FloatCurve=DataInterface() {propertyValues {Curve {keys [
          {interpMode=RCIM_Cubic, leaveTangent=-2.3646, tangentMode=RCTM_Break, time=0, value=1}
          <at 0.6689 0.0233>
          {arriveTangent=-0.2139, interpMode=RCIM_Cubic, tangentMode=RCTM_Break, time=1, value=0}
         ]}}})
     B=&FNTimeOfDayScalar(Night Scale=0.5, Morning Scale=0.75, Evening Scale=0.75)
    )
    ScaleA=true
    ScaleRGB=true
   ) {module=ScaleColor002}
   &ScaleColor(
    Linear Color Curve=DataInterface() {propertyValues {
      AlphaCurve {keys [<at 0 1>]}
      BlueCurve {keys [<at 0.0663 0.1061>,<at 0.4127 0.1091>]}
      GreenCurve {keys [<at 0.0663 0.074>,<at 0.4127 0.0855>]}
      RedCurve {keys [<at 0.0663 2>,<at 0.4127 1.5>]}
     }}
    Scale Mode=::NewEnumerator2
   )
  ]
  rendererClasses [&NiagaraRibbonRendererProperties]
  renderers [{rendererClass=&NiagaraRibbonRendererProperties}]
 }
 Trail_Fast {
  emitterUpdateScript [
   &EmitterState(Life Cycle Mode=::NewEnumerator1, Loop Behavior=::NewEnumerator1, Loop Delay=0.1, Loop Duration=0.1, UseLoopDelay=true)
   &FN_SpawnAnimTrail_v2()
  ]
  particleSpawnScript [
   &FN_InitializeAnimTrailBoneIndecies(Second Bone or Socket Index=1)
   &FN_InitializeAnimTrail(AnimTrailWidth=1.2, Lifetime=0.17, SkeletalMesh=$System.Skeletal Mesh) {module=FN_InitializeAnimTrail001}
  ]
  particleUpdateScript [
   &FN_UpdateAnimTrail(SkeletalMesh=$System.Skeletal Mesh)
   &SolveForcesAndVelocity(Clamp Velocity=false, Limit Acceleration=false)
   &ScaleColor(
    Scale Alpha=&FloatFromCurve(
     FloatCurve=DataInterface() {propertyValues {Curve {keys [{arriveTangent=-2.2741, interpMode=RCIM_Cubic, leaveTangent=-2.2741, tangentMode=RCTM_User, time=0, value=1}, <at 0.6 0 RCIM_Cubic>]}}}
     Scale Curve=&FNTimeOfDayScalar(Night Scale=0.3, Morning Scale=0.3, Day Scale=0.3, Evening Scale=0.3)
    )
    Scale RGB=&MakeVectorFromLinearColorRGB(Color=&ColorFromCurve(ColorCurve=DataInterface() {propertyValues {
        AlphaCurve {keys []}
        BlueCurve {keys [<at 0 0 RCIM_Cubic>,<at 0.2156 0.136>,<at 0.675 0.045 RCIM_Cubic>]}
        GreenCurve {keys [<at 0 0.75 RCIM_Cubic>,<at 0.2156 0.5104>,<at 0.675 0.1236 RCIM_Cubic>]}
        RedCurve {keys [<at 0 1.5 RCIM_Cubic>,<at 0.2156 1>,<at 0.675 0.281 RCIM_Cubic>]}
       }}))
    ScaleA=true
    ScaleRGB=true
   )
  ]
  rendererClasses [&NiagaraRibbonRendererProperties]
  renderers [{rendererClass=&NiagaraRibbonRendererProperties}]
 }
 Trail_Screen {
  emitterUpdateScript [
   &EmitterState(Life Cycle Mode=::NewEnumerator1, Loop Behavior=::NewEnumerator1, Loop Duration=0.27, UseLoopDelay=true)
   &FN_SpawnAnimTrail_v2(Number of Anim Trails=3)
  ]
  particleSpawnScript [
   &FN_InitializeAnimTrailBoneIndecies(First Bone or Socket Index=1, Second Bone or Socket Index=1)
   &FN_InitializeAnimTrailBoneIndecies(AnimTrail Selection=1, First Bone or Socket Index=2, Second Bone or Socket Index=2) {module=FN_InitializeAnimTrailBoneIndecies001}
   &FN_InitializeAnimTrailBoneIndecies(AnimTrail Selection=2, First Bone or Socket Index=3, Second Bone or Socket Index=3) {module=FN_InitializeAnimTrailBoneIndecies002}
   &FN_InitializeAnimTrail(
    AnimTrailWidth=10
    Color=<rgba 1 1 1 0.2>
    Lifetime=/Game/Effects/Niagara/DynamicInputs/FN_PickFloatFromSpawnGroupCutoff(
     Over Cutoff Float:NiagaraFloat=0.15
     Under Cutoff Float:NiagaraFloat=0.2
     SpawnGroup:NiagaraInt32=$Particles.SpawnGroup
     Cutoff:NiagaraInt32=1
    )
    SkeletalMesh=$System.Skeletal Mesh
   ) {module=FN_InitializeAnimTrail001}
  ]
  particleUpdateScript [
   &FN_UpdateAnimTrail(SkeletalMesh=$System.Skeletal Mesh)
   &SolveForcesAndVelocity(Clamp Velocity=false, Limit Acceleration=false)
   &ScaleRibbonWidth(Ribbon Width Scale=&FloatFromCurve(FloatCurve=DataInterface() {propertyValues {Curve {keys [
         {arriveTangent=-1.2579, interpMode=RCIM_Cubic, time=0, value=1}
         {arriveTangent=0.0144, interpMode=RCIM_Cubic, leaveTangent=0.0144, tangentMode=RCTM_User, time=1, value=1.8}
        ]}}}))
   &DynamicMaterialParameters(
    Index 0 Param 1=&RandomRangeFloat(Minimum=-0.25, Maximum=0.25)
    Index 0 Param 2=1
    Index 0 Param 3=-0.8
    Index 0 Param 4=&FloatFromCurve(FloatCurve=DataInterface() {propertyValues {Curve {keys [
         {interpMode=RCIM_Cubic, tangentMode=RCTM_User, time=0, value=1}
         {arriveTangent=-0.6219, interpMode=RCIM_Cubic, leaveTangent=-0.6219, tangentMode=RCTM_User, time=1, value=0}
        ]}}})
    Module.Param0WriteEnabled=true
    Module.Param1WriteEnabled=true
    Module.Param2WriteEnabled=false
    Module.Param3WriteEnabled=false
   )
   &Color(
    Color=&ColorFromCurve(ColorCurve=DataInterface() {propertyValues {
       AlphaCurve {keys [<at 0.241 1>,<at 1 0 RCIM_Cubic>]}
       BlueCurve {keys [
         <at 0.006 0.321>
         {arriveTangent=0.0837, interpMode=RCIM_Cubic, leaveTangent=0.0837, time=0.2498, value=0.6972}
         {arriveTangent=-0.3636, interpMode=RCIM_Cubic, leaveTangent=-0.3636, time=0.5378, value=0.3655}
         <at 0.8696 0.4718>
        ]}
       GreenCurve {keys [
         <at 0.006 0.18>
         {arriveTangent=-0.1843, interpMode=RCIM_Cubic, leaveTangent=-0.1843, time=0.2498, value=0.162}
         {arriveTangent=-0.1204, interpMode=RCIM_Cubic, leaveTangent=-0.1204, time=0.5378, value=0.082}
         <at 0.8696 0.0873>
        ]}
       RedCurve {keys [
         <at 0.006 3>
         {arriveTangent=-4.2414, interpMode=RCIM_Cubic, leaveTangent=-4.2414, time=0.2498, value=1.5}
         {arriveTangent=-1.4519, interpMode=RCIM_Cubic, leaveTangent=-1.4519, time=0.5378, value=0.7448}
         <at 0.8696 0.6>
        ]}
      }})
    Scale Alpha=&FNTimeOfDayScalar(Night Scale=0.25, Morning Scale=0.35, Day Scale=0.5, Evening Scale=0.35)
   ) {enabled=false}
   &Color(
    Color=&ColorFromCurve(ColorCurve=DataInterface() {propertyValues {
       AlphaCurve {keys [<at 0.241 1>,<at 1 0 RCIM_Cubic>]}
       BlueCurve {keys [
         <at 0.006 0.4667>
         {arriveTangent=-0.5024, interpMode=RCIM_Cubic, leaveTangent=-0.5024, time=0.1961, value=0.1384}
         {arriveTangent=-0.5024, interpMode=RCIM_Cubic, leaveTangent=-0.5024, time=0.4724, value=0.0839}
         <at 0.8696 0.0328>
        ]}
       GreenCurve {keys [<at 0.006 0>,<at 0.1961 0 RCIM_Cubic>,<at 0.4724 0 RCIM_Cubic>,<at 0.8696 0>]}
       RedCurve {keys [
         <at 0.006 2>
         {arriveTangent=-2.153, interpMode=RCIM_Cubic, leaveTangent=-2.153, time=0.1961, value=0.593}
         {arriveTangent=-2.153, interpMode=RCIM_Cubic, leaveTangent=-2.153, time=0.4724, value=0.3594}
         <at 0.8696 0.1406>
        ]}
      }})
    Scale Alpha=&FNTimeOfDayScalar(Night Scale=0.5, Morning Scale=1, Evening Scale=1)
   ) {module=Color001}
  ]
  rendererClasses [&NiagaraRibbonRendererProperties]
  renderers [{rendererClass=&NiagaraRibbonRendererProperties}]
 }
}
summary {
 systemName=NS_Pickaxe_LiftingAura_Trail_Variant
 emitters [
  {emitterName=Dark001, rendererClasses [&NiagaraRibbonRendererProperties]}
  {emitterName=Trail, rendererClasses [&NiagaraRibbonRendererProperties]}
  {emitterName=Trail_Fast, rendererClasses [&NiagaraRibbonRendererProperties]}
  {emitterName=Trail002, rendererClasses [&NiagaraRibbonRendererProperties]}
  {emitterName=Trail_Screen, rendererClasses [&NiagaraRibbonRendererProperties]}
  {emitterName=Gold_Dust, rendererClasses [&NiagaraSpriteRendererProperties]}
  {emitterName=Snowflakes002, rendererClasses [&NiagaraSpriteRendererProperties]}
 ]
 userVariables [User.RibbonIDOffset {type=NiagaraInt32}]
}
